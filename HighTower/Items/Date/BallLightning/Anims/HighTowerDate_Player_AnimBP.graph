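# Player item-layer AnimBlueprint for HighTower Date's Ball Lightning, supplying a Wait, Intro, Loop, Cast state machine in ItemFinalPoseOverride.
# Each update matches the active montage by name to start Intro and checks the cooldown gameplay tag to trigger Cast.
# It also smooths a UserCurve01-driven FullBodyAlpha and debug-prints the curve.
# The final blend's alpha shows as a constant 0, favoring the input pose, and some event logic is disconnected.
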AnimBlueprint /HighTower/Items/Date/BallLightning/Anims/HighTowerDate_Player_AnimBP
target_skeleton=&Fortnite_M_Avg_Player_Skeleton
graphs [
 AnimGraph {nodes [Output Pose()]}
 EventGraph {nodes [
   n1 = Event BlueprintUpdateAnimation()
   > Set Delta Time X(Delta Time X=@n1.DeltaTimeX)
   > n2 = Sequence()
   n3 = Get FortPlayerPawn() -> (FortPlayerPawn:Fort Player Pawn Object Reference)
   n4 = GetCurrentActiveMontage!(self:Anim Instance Object Reference=GetAnimInstance(self=Get Mesh(self=@n3))) -> (ReturnValue:Anim Montage Object Reference)
   > Set IsLightningBall(IsLightningBall:bool="Equal Exactly (String)"(A="To String (Object)"!(InObj:Object Reference=@n4) -> (ReturnValue:String), B=HighTowerDate_Ball_Cast_Montage)) -> (Output_Get:bool)
   n5 = GetCurveValue(CurveName=UserCurve01)
   n6 = "To Float (Boolean)"(InBool="float > float"(A=@n5, B=0.5))
   > Set FullBodyAlpha(FullBodyAlpha:float=FInterpTo(
     Current=Get FullBodyAlpha() -> (FullBodyAlpha:float)
     Target="Clamp (Float)"(Value=@n6, Min=0.01, Max=0.99)
     DeltaTime=Get Delta Time X()
     InterpSpeed=6
    )) -> (Output_Get:float)
   > Set IsCast(IsCast:bool=HasMatchingGameplayTag!(
     self:Gameplay Tag Asset Interface Interface=Get FortPlayerPawn()
     TagToCheck:Gameplay Tag Structure=(TagName="Granted.Cooldown.HighTower.Date.BallLightning")
    ) -> (ReturnValue:bool)) -> (Output_Get:bool)
   PrintString!(
    execute=@n2.then_1
    InString="To String (Float)"(InDouble=@n5)
    bPrintToScreen
    bPrintToLog
    TextColor=(R=1,G=0.9126,B=0,A=1)
    Duration=0.07
    Key
   ) -> (then)
   n7 = Branch()
   Set IsFullBody(execute=@n7.else, IsFullBody:bool=false) -> (Output_Get:bool)
   > Set IsLoop(IsLoop:bool=false) -> (Output_Get:bool)
   > Set IsCast(IsCast=false)
   > Set FullBodyAlpha(FullBodyAlpha=0)
   n8 = Get Mesh(self=Get FortPlayerPawn())
   n9 = Montage_GetCurrentSection!(self:Anim Instance Object Reference, Montage:Anim Montage Object Reference=GetCurrentActiveMontage(self=GetAnimInstance(self=@n8))) -> (ReturnValue:Name=None)
   Set IsLoop(IsLoop="Equal (Name)"(A=@n9, B=Loop))
   > Set IsCast(IsCast="Equal (Name)"(A=@n9, B=Cast))
   Set IsFullBody(IsFullBody=false)
  ]}
 ItemAimOffset {nodes [
   Input Pose!() -> (Pose, AimOffsetAlpha:double, Yaw:double, Pitch:double) {name=InPose_AimOffset}
   Input Pose!() -> (Pose) {name=InPose_PreAimOffsets}
   Output Pose() {layerGroup=Item Override}
  ]}
 ItemFallAdditive {nodes [
   Input Pose() {name=InPoseItemFallAdditive}
   Output Pose() {layerGroup=Item Override}
  ]}
 ItemFallLandAdditive {nodes [
   Input Pose() {name=InPoseItemFallLandAdditive}
   Output Pose() {layerGroup=Item Override}
  ]}
 ItemFallLevel2Override {nodes [
   Input Pose() {name=InPose_FallLevel2}
   Output Pose() {layerGroup=Item Override}
  ]}
 ItemFinalPoseOverride {nodes [
   AnimGraphNode_StateMachine_0 = BallLightning SM() {is=title AnimGraphNode_StateMachine}
   Output Pose(Result=Two Way Blend(A=Use cached pose 'Final Input Pose'(), B=@AnimGraphNode_StateMachine_0, Alpha=0) {blendNode {alphaBoolBlend {blendOption=Linear}}}) {layerGroup=Item Override}
   Save cached pose 'Final Input Pose'(Pose=Input Pose() {name=InFinalPose}) {cacheName=Final Input Pose}
  ]}
 BallLightning SM {
  entry=Wait
  path=ItemFinalPoseOverride.AnimGraphNode_StateMachine_0.BallLightning SM
  ids=id
  nodes [
   AnimStateNode_0 = Intro() {is=title AnimStateNode}
   AnimStateNode_1 = Loop() {is=title AnimStateNode}
   AnimStateNode_2 = Cast() {is=title AnimStateNode}
   AnimStateNode_3 = Wait() {is=title AnimStateNode}
   AnimStateTransitionNode_0 = AnimStateTransitionNode() {bAutomaticRuleBasedOnSequencePlayerInState=true, crossfadeDuration=0}
   AnimStateTransitionNode_1 = AnimStateTransitionNode() {crossfadeDuration=0}
   AnimStateTransitionNode_2 = AnimStateTransitionNode()
   AnimStateTransitionNode_3 = AnimStateTransitionNode() {bAutomaticRuleBasedOnSequencePlayerInState=true}
  ]
  links [
   {
    dst=Wait
    rule=ItemFinalPoseOverride.AnimGraphNode_StateMachine_0.BallLightning SM.AnimStateTransitionNode_3.Transition
    src=Cast
    via=AnimStateTransitionNode_3
   }
   {
    dst=Loop
    rule=ItemFinalPoseOverride.AnimGraphNode_StateMachine_0.BallLightning SM.AnimStateTransitionNode_0.Transition
    src=Intro
    via=AnimStateTransitionNode_0
   }
   {
    dst=Cast
    rule=ItemFinalPoseOverride.AnimGraphNode_StateMachine_0.BallLightning SM.AnimStateTransitionNode_1.Transition
    src=Loop
    via=AnimStateTransitionNode_1
   }
   {
    dst=Intro
    rule=ItemFinalPoseOverride.AnimGraphNode_StateMachine_0.BallLightning SM.AnimStateTransitionNode_2.Transition
    src=Wait
    via=AnimStateTransitionNode_2
   }
  ]
 }
 Intro {
  path=ItemFinalPoseOverride.AnimGraphNode_StateMachine_0.BallLightning SM.AnimStateNode_0.Intro
  nodes [Output Animation Pose(Result=Sequence Player 'HighTowerDate_Combo_Bomb_Intro'() {bLoopAnimation=false, sequence=/HighTower/Animation/Date/Combo/HighTowerDate_Combo_Bomb_Intro})]
 }
 Loop {
  path=ItemFinalPoseOverride.AnimGraphNode_StateMachine_0.BallLightning SM.AnimStateNode_1.Loop
  nodes [Output Animation Pose(Result=Sequence Player 'HighTowerDate_Combo_Bomb_Loop'() {sequence=/HighTower/Animation/Date/Combo/HighTowerDate_Combo_Bomb_Loop})]
 }
 Cast {
  path=ItemFinalPoseOverride.AnimGraphNode_StateMachine_0.BallLightning SM.AnimStateNode_2.Cast
  nodes [Output Animation Pose(Result=Sequence Player 'HighTowerDate_Combo_Bomb_Throw'() {bLoopAnimation=false, sequence=/HighTower/Animation/Date/Combo/HighTowerDate_Combo_Bomb_Throw})]
 }
 Wait {
  path=ItemFinalPoseOverride.AnimGraphNode_StateMachine_0.BallLightning SM.AnimStateNode_3.Wait
  nodes [Output Animation Pose(Result=Use cached pose 'Final Input Pose'())]
 }
 Transition {path=ItemFinalPoseOverride.AnimGraphNode_StateMachine_0.BallLightning SM.AnimStateTransitionNode_0.Transition, nodes [Result()]}
 Transition {
  path=ItemFinalPoseOverride.AnimGraphNode_StateMachine_0.BallLightning SM.AnimStateTransitionNode_1.Transition
  nodes [Result(bCanEnterTransition=Get IsCast() -> (IsCast:bool))]
 }
 Transition {
  path=ItemFinalPoseOverride.AnimGraphNode_StateMachine_0.BallLightning SM.AnimStateTransitionNode_2.Transition
  nodes [Result(bCanEnterTransition=Get IsLightningBall() -> (IsLightningBall:bool))]
 }
 Transition {path=ItemFinalPoseOverride.AnimGraphNode_StateMachine_0.BallLightning SM.AnimStateTransitionNode_3.Transition, nodes [Result()]}
 ItemFullBodyAimOffsets {nodes [
   Input Pose!() -> (Pose, FullBody_AimOffsetAlpha:double, Aim_Yaw:double, Aim_Pitch:double) {name=InPoseFullBodyAimOffsets}
   Input Pose!() -> (Pose) {name=InPosePreFullBodyAimOffsets}
   Output Pose() {layerGroup=Item Override}
  ]}
 ItemFullBodyOverride {nodes [
   Input Pose() {name=InPoseFullBody}
   Output Pose() {layerGroup=Item Override}
  ]}
 ItemIdleOverride {nodes [
   Input Pose() {name=InPose_Idle}
   Output Pose() {layerGroup=Item Override}
  ]}
 ItemJetPackJumpAdditive {nodes [Output Pose() {layerGroup=Item Override}]}
 ItemJetPackStartAdditive {nodes [Output Pose() {layerGroup=Item Override}]}
 ItemJumpLoopAdditive {nodes [
   Input Pose() {name=InPoseItemJumpLoopAdditive}
   Output Pose() {layerGroup=Item Override}
  ]}
 ItemJumpOverride {nodes [
   Input Pose() {name=InPoseJump}
   Output Pose() {layerGroup=Item Override}
  ]}
 ItemJumpUpAdditive {nodes [
   Input Pose() {name=InPoseItemJumpUpAdditive}
   Output Pose() {layerGroup=Item Override}
  ]}
 ItemPreAimOffsets {nodes [
   Input Pose() {name=InPosePreAimOffsets}
   Output Pose() {layerGroup=Item Override}
  ]}
 ItemPreIK {nodes [
   Input Pose() {name=InPose_PreIK}
   Output Pose() {layerGroup=Item Override}
  ]}
 ItemRightHandIK {nodes [
   Input Pose() {name=InPosePreRightHandIK}
   Input Pose() {name=InPosePostRightHandIK}
   Output Pose() {layerGroup=Item Override}
  ]}
 ItemSkeletalControl {nodes [
   Input Pose()
   Output Pose() {layerGroup=Item Override}
  ]}
 ItemSwimJumpFallLoop {nodes [Output Pose() {layerGroup=Item Override}]}
 ItemSwimJumpSurfaceEnd {nodes [Output Pose() {layerGroup=Item Override}]}
 ItemSwimJumpSurfaceLoop {nodes [Output Pose() {layerGroup=Item Override}]}
 ItemUpperBody {nodes [
   Input Pose!() -> (Pose, InputParam:Fort Anim Input Adjusted Aim Structure) {name=InPoseUpperBody}
   Output Pose() {layerGroup=Item Override}
  ]}
 ItemZipLineStartAdditive {nodes [Output Pose() {layerGroup=Item Override}]}
]
variables [
 Delta Time X {type=number}
 FullBodyAlpha {type=number}
 IsCast {type=boolean}
 IsFullBody {type=boolean}
 IsLightningBall {type=boolean}
 IsLoop {type=boolean}
]
