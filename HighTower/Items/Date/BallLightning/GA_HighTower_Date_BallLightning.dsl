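; Ball Lightning ability for the HighTower Date item: a hold-and-release throw that spawns a projectile through its parent power class.
; Release waits out a minimum hold time by timer, and the throw fires once both release and the montage's ok-to-throw notify have occurred.
; Throwing forces a net update, and projectile spawn removes the held-input loop cue to avoid timing issues.
; Ghost-pull and dance-stun tag watches accompany a net-synced early end that skips cooldown.
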(event Ability|EventActivateAbility
  (Utilities|FlowControl|Sequence
    (:then_0
      ;[1 1 1] List of player caused effects that should end the ability
      (Ability|Tasks|WaitGameplayTagAdd "(TagName=\"Granted.Athena.HappyGhost.Pull\")"
        (:then
          (Ability|Tasks|WaitGameplayTagAdd "(TagName=\"Gameplay.Action.Player.DanceStunned\")"
            (:Added
              (Development|PrintString "DanceStunned" false false))))
        (:Added
          (Development|PrintString "ghost pulled" false false)))
      (Ability|Tasks|WaitNetSync "BothWait"
        (:OnSync
          ;[1 1 1] Note - ability ending this way does not take cooldown
          (Animation|StopAnimMontage (Variables|Default|GetPlayerPawn) (Variables|PowerAnimationCast|GetCastMontage))
          (Ability|EndAbility))))
    (:then_1
      (Default|BindEventtoOnProjectileSpawned (delegate Custom|CustomEvent_2))
      (|CallParentFunction))))

(event Custom|CustomEvent_2
  (params (ProjectileSpawned :type Fort%20Projectile%20Base%20Object%20Reference))
  ;[1 1 1] Tons of timing issues with this being removed any other way
  (Ability|RemoveGameplayCueFromOwner (Variables|PowerGameplayCues|GetGC_InputHeldLoop)))

(event Custom|CustomEvent_1
  ;[1 1 1] If we release the button before throw is ready, we set it up for the notify to handle the projectile spawning
  (Variables|Default|SetInputReleased true)
  (if (Variables|Default|GetFiringAllowed)
    (CallFunction|MulticastThrow)))

(event EventReleased
  (params (HoldTime :type Float%20%28double-precision%29))
  ;[1 1 1] If we release the button before throw is ready, we set it up for the notify to handle the projectile spawning
  (bind (_success _mintimeremaining) (CallFunction|CheckMinHoldTime))
  (if _success
    (Variables|Default|SetInputReleased true)
    (if (Variables|Default|GetFiringAllowed)
      (CallFunction|MulticastThrow))
    (else
      (Utilities|Time|SetTimerbyEvent (delegate Custom|CustomEvent_1) _mintimeremaining))))

(event EventCastMontageNotifyStart
  (params (NotifyTag :type Gameplay%20Tag%20Structure))
  ;[1 1 1] If we reach the "ok to throw" notify while the input is  still held, wait for the player to release the button to spawn projectile
  (if (GameplayTags|Equal(GameplayTag) NotifyTag (Variables|Default|GetT_OkToThrowNotify))
    (Variables|Default|SetFiringAllowed true)
    (if (Variables|Default|GetInputReleased)
      (CallFunction|MulticastThrow))
    (else
      (|CallParentFunction NotifyTag))))

(event Ability|EventOnEndAbility
  (params (bWasCancelled :type Boolean))
  ;[1 1 1] Reset ability vars
  (|CallParentFunction bWasCancelled)
  (Variables|Default|SetInputReleased false)
  (Variables|Default|SetFiringAllowed false))

(event Custom|MulticastThrow
  ;[1 1 1] Multicast the removal of GC and force a net update so remote clients see the animation.  Spawn the projectile.
  (Networking|ForceNetUpdate (Variables|Default|GetPlayerPawn))
  (CallFunction|OnAbilityActivated))

(fn CheckMinHoldTime
  (return (> (- (Utilities|Time|GetGameTimeinSeconds) (Variables|Default|GetHoldStartTime)) (Variables|Default|GetMinHoldTime)) (- (Variables|Default|GetMinHoldTime) (- (Utilities|Time|GetGameTimeinSeconds) (Variables|Default|GetHoldStartTime)))))

(fn InitAbility
  (|CallParentFunction)
  (Variables|Default|SetHoldStartTime (Utilities|Time|GetGameTimeinSeconds)))

(fn CanActivateAbility
  (params (ActorInfo :type Gameplay%20Ability%20Actor%20Info%20Structure) (Handle :type Gameplay%20Ability%20Spec%20Handle%20Structure))
  (bind (_relevanttags _returnvalue) (|CallParentFunction ActorInfo Handle))
  (return _relevanttags _returnvalue))

(fn PowerSpawnProjectile
  (bind _projectilespawned (|CallParentFunction))
  (return _projectilespawned))
